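# Niagara impact effect for the StarWandFNCS pickaxe: a short burst of spinning star sparks around a flash.
# Stars_Sparks fires two GPU sprite bursts outward from a sphere, using spawn groups to vary size, then slows them with gravity and drag.
# spine_Center adds a brief, randomly oriented mesh and fading light.
# Time-of-day inputs tune color and brightness, front-end modifiers shrink and slow particles; glow, corona and streak emitters show no details.
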NiagaraSystem /BRCosmetics/Weapons/FORT_Melee/Pickaxe_StarWandFNCS/NS_Pickaxe_StarWandFNCS_Impact
dependencies {
 usedDataInterfaces [&NiagaraDataInterfaceArrayColor, &NiagaraDataInterfaceCurve, &NiagaraDataInterfaceColorCurve]
 usedDynamicInputs [
  &RandomRangeLinearColor
  &Vector2DFromFloat
  &FN_FrontEndVector2Modifier
  &RandomRangeVector2D
 ]
 usedModules [
  &self:SystemSpawnScript.SystemScriptSource.SystemScriptGraph.NiagaraNodeAssignment_0.SetVariables_#1
  &self:SystemSpawnScript.SystemScriptSource.SystemScriptGraph.NiagaraNodeAssignment_1.SetVariables_#2
  &SystemState
 ]
}
emitters {
 CoronaLWE {rendererClasses [], renderers []}
 GlowBlackLWE {rendererClasses [], renderers []}
 GlowLWE {rendererClasses [], renderers []}
 StarEnergy {rendererClasses [], renderers []}
 StarGlow {rendererClasses [], renderers []}
 Stars_Sparks {
  simTarget=GPUComputeSim
  emitterUpdateScript [
   &EmitterState()
   &SpawnBurst_Instantaneous(
    Spawn Count=&RandomRangeInt(Minimum=12, Maximum=15)
    Use Spawn Probability=false
    UseLoopCountLimit=false
   )
   &SpawnBurst_Instantaneous(
    Spawn Count=&RandomRangeInt(Minimum=8, Maximum=12)
    Spawn Group=1
    Use Spawn Probability=false
    UseLoopCountLimit=false
   ) {module=SpawnBurst_Instantaneous001}
  ]
  particleSpawnScript [
   /SetVariables(Particles.VisibilityTag=&SpawnGroup()) {isSetParameters=true}
   &InitializeParticle(
    Color=&FNTimeOfDayColor(
     Night Color=&SelectLinearColorFromArray(
      Array Sampling Mode=::NewEnumerator0
      Direct Array Index=?
      Color Selection Array=DataInterface() {propertyValues {ColorData [
         <rgba 0 0 0 1>
         <rgba 0.1925 0.2515 0.35 1>
         <rgba 0.15 0.0127 0.0113 1>
         <rgba 0.76 0.4836 0.057 1>
         <rgba 0.15 0.0127 0.0113 1>
         <rgba 0.76 0.4836 0.057 1>
         <rgba 0.15 0.0127 0.0113 1>
         <rgba 0.76 0.4836 0.057 1>
        ]}}
      Random Seed=0
     )
     Morning Color=&SelectLinearColorFromArray(
      Array Sampling Mode=::NewEnumerator0
      Direct Array Index=?
      Color Selection Array=DataInterface() {propertyValues {ColorData [
         <rgba 0 0 0 1>
         <rgba 1 1 1 1>
         <rgba 1 0 0.0375 1>
         <rgba 1 0.4668 0 1>
         <rgba 1 0 0.0375 1>
         <rgba 1 0.4668 0 1>
         <rgba 1 0 0.0375 1>
         <rgba 1 0.4668 0 1>
        ]}}
      Random Seed=0
     )
     Day Color=&SelectLinearColorFromArray(
      Array Sampling Mode=::NewEnumerator0
      Direct Array Index=?
      Color Selection Array=DataInterface() {propertyValues {ColorData [
         <rgba 0 0 0 1>
         <rgba 1 1 1 1>
         <rgba 1 0 0.0375 1>
         <rgba 1 0.4668 0 1>
         <rgba 1 0 0.0375 1>
         <rgba 1 0.4668 0 1>
         <rgba 1 0 0.0375 1>
         <rgba 1 0.4668 0 1>
        ]}}
      Random Seed=0
     )
     Evening Color=&SelectLinearColorFromArray(
      Array Sampling Mode=::NewEnumerator0
      Direct Array Index=?
      Color Selection Array=DataInterface() {propertyValues {ColorData [
         <rgba 0 0 0 1>
         <rgba 1 1 1 1>
         <rgba 1 0 0.0375 1>
         <rgba 1 0.4668 0 1>
         <rgba 1 0 0.0375 1>
         <rgba 1 0.4668 0 1>
         <rgba 1 0 0.0375 1>
         <rgba 1 0.4668 0 1>
        ]}}
      Random Seed=0
     )
    )
    Lifetime Max=0.8
    Lifetime Min=0.3
    Lifetime Mode=::NewEnumerator1
    Sprite Rotation Mode=::NewEnumerator2
    Sprite Size Mode=::NewEnumerator3
    Uniform Sprite Size=&FN_FrontEndFloatModifier(
     Gameplay=&MaskFloatBySpawnGroup(
      Masked Float=&RandomRangeFloat(Minimum=10, Maximum=30)
      Passthrough Float=&RandomRangeFloat(Minimum=5, Maximum=12)
     )
     Front End=&MaskFloatBySpawnGroup(
      Masked Float=&RandomRangeFloat(Minimum=6, Maximum=18)
      Passthrough Float=&RandomRangeFloat(Minimum=2, Maximum=8)
     )
    )
    UsePositionOffset=false
    Write Material Random=true
   )
   &ShapeLocation(
    MaskSpawnGroup=false
    Offset Mode=::NewEnumerator0
    Override Seed=false
    Sphere Radius=&MaskFloatBySpawnGroup(Masked Float=45, Passthrough Float=50)
    UseOwnerScale=true
   )
   &AddVelocity(
    Module.Use Falloff Distance=false
    Override Seed=false
    Velocity Mode=::NewEnumerator1
    Velocity Origin=$Particles.ShapeLocation.ShapePosition
    Velocity Speed=&FN_FrontEndFloatModifier(
     Gameplay=&RandomRangeFloat(Minimum=250, Maximum=600)
     Front End=&RandomRangeFloat(Minimum=85, Maximum=250)
    )
   )
  ]
  particleUpdateScript [
   &ParticleState()
   &SpriteRotationRate(Rotation Rate=&RandomRangeFloat(Minimum=3, Maximum=5))
   &ScaleSpriteSize(Uniform Curve Sprite Scale=DataInterface() {propertyValues {Curve {
       defaultValue=1
       keys [
        {arriveTangent=9.0725, interpMode=RCIM_Cubic, leaveTangent=9.0725, tangentMode=RCTM_User, time=0, value=0}
        {arriveTangent=-2.5249, interpMode=RCIM_Cubic, leaveTangent=-2.5249, tangentMode=RCTM_User, time=0.1408, value=0.9847}
        {arriveTangent=-0.0405, interpMode=RCIM_Cubic, leaveTangent=-0.0405, tangentMode=RCTM_User, time=1, value=0}
       ]
      }}})
   &ScaleColor(
    Linear Color Curve=DataInterface() {propertyValues {
      AlphaCurve {keys [<at 0 0 RCIM_Cubic>,<at 0.2 1 RCIM_Cubic>,<at 1 0 RCIM_Cubic>]}
      BlueCurve {keys [<at 0 1>,<at 1 0>]}
      GreenCurve {keys [<at 0 1>,<at 1 0>]}
      RedCurve {keys [<at 0 1>,<at 1 1>]}
     }}
    Scale Mode=::NewEnumerator2
   )
   &CurlNoiseForce(Bias Noise=false, Noise Strength=&RandomRangeFloat(Minimum=200, Maximum=300), Pan Noise=false) {enabled=false}
   &GravityForce(Gravity=&MakeVector(X=0, Y=0, Z=&FN_FrontEndFloatModifier(Gameplay=-980, Front End=-750)))
   &Drag(
    Drag=&FloatFromCurve(
     FloatCurve=DataInterface() {propertyValues {Curve {keys [<at 0 0>,<at 0.1 1>]}}}
     Scale Curve=&RandomRangeFloat(Minimum=3, Maximum=5)
    )
    Use Linear Drag=true
    Use Rotational Drag=true
   )
   &SolveForcesAndVelocity(Clamp Velocity=false, Limit Acceleration=false)
  ]
  rendererClasses [&NiagaraSpriteRendererProperties]
  renderers [
   {rendererClass=&NiagaraSpriteRendererProperties}
   {rendererClass=&NiagaraSpriteRendererProperties, rendererIndex=1}
  ]
 }
 StreaksBlackLWE {rendererClasses [], renderers []}
 StreaksLWE {rendererClasses [], renderers []}
 StreaksWhiteLWE {rendererClasses [], renderers []}
 spine_Center {
  emitterUpdateScript [
   &EmitterState()
   &SpawnBurst_Instantaneous(Use Spawn Probability=false, UseLoopCountLimit=false)
  ]
  particleSpawnScript [
   &InitializeParticle(
    Color=&MultiplyLinearColor(
     Linear Color=<rgba 1.25 0.8004 0.3906 1>
     Scale Factor=&MakeVector4_FromLinearColor(LinearColor=&FNTimeOfDayColor(
       Night Color=<rgba 0.4086 0.4973 0.6 0.4>
       Morning Color=<rgba 0.875 0.9329 1 1>
       Day Color=<rgba 1 1 1 1>
       Evening Color=<rgba 0.875 0.9329 1 1>
      ))
    )
    Lifetime=0.15
    Mesh Scale Mode=::NewEnumerator3
    Mesh Uniform Scale=&FN_FrontEndFloatModifier(
     Gameplay=&RandomRangeFloat(Minimum=0.7, Maximum=0.9)
     Front End=&RandomRangeFloat(Minimum=0.4, Maximum=0.6)
    )
    UsePositionOffset=false
    Write Material Random=true
   )
   &InitialMeshOrientation(
    Mesh Orientation Mode=::NewEnumerator2
    Rotation=&RandomRangeVector(Minimum=<-1 -1 -1>, Maximum=<1 1 1>)
    Use Rotation Vector=true
   )
   &CameraOffset()
  ]
  particleUpdateScript [
   &ParticleState()
   &ScaleMeshSize(Scale Factor=&VectorFromFloat(Value=&FloatFromCurve(FloatCurve=DataInterface() {propertyValues {Curve {keys [
          {interpMode=RCIM_Cubic, tangentMode=RCTM_User, time=0, value=1}
          {arriveTangent=-0.5799, interpMode=RCIM_Cubic, leaveTangent=-4.8336, tangentMode=RCTM_Break, time=0.6496, value=0.8318}
          {arriveTangent=0.0139, interpMode=RCIM_Cubic, leaveTangent=0.0139, tangentMode=RCTM_User, time=1, value=0}
         ]}}})))
   &DynamicMaterialParameters(
    Index 0 Param 1=&FloatFromCurve(FloatCurve=DataInterface() {propertyValues {Curve {keys [
         {arriveTangent=0.0062, interpMode=RCIM_Cubic, leaveTangent=0.0062, tangentMode=RCTM_User, time=0, value=0}
         <at 0.9 0>
         {arriveTangent=0.0152, interpMode=RCIM_Cubic, leaveTangent=0.0152, tangentMode=RCTM_User, time=1, value=0.5}
        ]}}})
    Index 0 Param 2=1
    Index 0 Param 3=1
    Index 0 Param 4=1
    Module.Param0WriteEnabled=true
    Module.Param1WriteEnabled=true
    Module.Param2WriteEnabled=true
    Module.Param3WriteEnabled=true
   )
   &ScaleColor(
    Scale Alpha=0.4
    Scale RGB=&MakeVectorFromLinearColorRGB(Color=&ColorFromCurve(ColorCurve=DataInterface() {propertyValues {
        AlphaCurve {keys [<at 0 1 RCIM_Cubic>,<at 1 0 RCIM_Cubic>]}
        BlueCurve {keys [<at 0.154 1.5 RCIM_Cubic>,<at 0.3971 0.8 RCIM_Cubic>]}
        GreenCurve {keys [<at 0.154 1.5 RCIM_Cubic>,<at 0.3971 0.8 RCIM_Cubic>]}
        RedCurve {keys [<at 0.154 1.5 RCIM_Cubic>,<at 0.3971 0.8 RCIM_Cubic>]}
       }}))
    ScaleA=true
    ScaleRGB=true
   )
   &FN_Light_Attributes(
    Color=&ColorFromCurve(ColorCurve=DataInterface() {propertyValues {
       AlphaCurve {keys []}
       BlueCurve {keys [<at 0 1 RCIM_Cubic>,<at 0.25 0>,<at 0.5 0>,<at 1 0 RCIM_Cubic>]}
       GreenCurve {keys [<at 0 1 RCIM_Cubic>,<at 0.25 0.784>,<at 0.5 0.356>,<at 1 0 RCIM_Cubic>]}
       RedCurve {keys [<at 0 1 RCIM_Cubic>,<at 0.25 2>,<at 0.5 10>,<at 1 0 RCIM_Cubic>]}
      }})
    Create Color Attribute=true
    Create Exponent Attribute=true
    Create Radius Attribute=true
    Exponent=4
    Radius=120
    Scale Color=&FloatFromCurve(
     FloatCurve=DataInterface() {propertyValues {Curve {keys [<at 0 1 RCIM_Cubic>,<at 1 0 RCIM_Cubic>]}}}
     Scale Curve=&FNTimeOfDayScalar(Night Scale=7, Morning Scale=9, Day Scale=15, Evening Scale=9)
    )
   )
  ]
  rendererClasses [&NiagaraMeshRendererProperties, &NiagaraLightRendererProperties]
  renderers [
   {rendererClass=&NiagaraMeshRendererProperties, rendererIndex=0}
   {rendererClass=&NiagaraLightRendererProperties, rendererIndex=1}
  ]
 }
}
summary {
 systemName=NS_Pickaxe_StarWandFNCS_Impact
 emitters [
  {emitterName=Stars_Sparks, simTarget=GPUComputeSim, rendererClasses [&NiagaraSpriteRendererProperties]}
  {emitterName=spine_Center, rendererClasses [&NiagaraMeshRendererProperties, &NiagaraLightRendererProperties]}
  {emitterName=GlowLWE, rendererClasses []}
  {emitterName=CoronaLWE, rendererClasses []}
  {emitterName=StarGlow, rendererClasses []}
  {emitterName=StarEnergy, rendererClasses []}
  {emitterName=StreaksLWE, rendererClasses []}
  {emitterName=GlowBlackLWE, rendererClasses []}
  {emitterName=StreaksBlackLWE, rendererClasses []}
  {emitterName=StreaksWhiteLWE, rendererClasses []}
 ]
 userVariables [
  User.bIsFrontEnd {type=NiagaraBool}
  User.bIsFrontEndPreview {type=NiagaraBool}
 ]
}
